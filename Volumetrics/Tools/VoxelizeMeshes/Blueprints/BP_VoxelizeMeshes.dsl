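; Actor Blueprint that voxelizes listed actors' meshes into 2D slice-atlas volume render targets, optionally taking resolution from an owning fluid simulation.
; Orthographic captures from several directions, or one capture of offset poseable-mesh copies, feed a composite material, directly or via an atlas.
; UV-unwrap captures of current and previous world positions supply motion vectors to the mesh materials.
; Optional jump flooding builds a distance-field gradient for the raymarch preview.
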(event Custom|Create%20RTs
  (bind _voxelization_rts (Variables|InternalRenderTargets|GetVoxelization%20RTs))
  ;[1 1 1] Create Render Targets
  (Utilities|AllocateRTIfNullOrChanged (Variables|InternalRenderTargets|GetUnwrapRTA) (Utilities|InttoIntPoint (Variables|VoxelizationSetup|GetScene%20Capture%20Resolution)) "RTF_RGBA16f"
    (:Out
      (Utilities|AllocateRTIfNullOrChanged (Variables|InternalRenderTargets|GetUnwrapRTB) (Utilities|InttoIntPoint (Variables|VoxelizationSetup|GetScene%20Capture%20Resolution)) "RTF_RGBA16f"
        (:Out
          (Utilities|AllocateRTIfNullOrChanged (Variables|InternalRenderTargets|GetVoxelCaptureFrameRT) (Utilities|InttoIntPoint (select (Variables|VoxelizationSetup|GetUse%20Poseable%20Meshes) (* (Variables|VoxelizationSetup|GetScene%20Capture%20Resolution) 2) (Variables|VoxelizationSetup|GetScene%20Capture%20Resolution))) "RTF_RGBA16f"
            (:Out
              (Utilities|FlowControl|Sequence
                (:then_0
                  (if (and (Variables|VoxelizationSetup|GetAtlas%20Voxelization) (not (Variables|VoxelizationSetup|GetUse%20Poseable%20Meshes)))
                    (Utilities|AllocateRTIfNullOrChanged (Variables|InternalRenderTargets|GetVoxelCaptureAtlasRT) (Utilities|InttoIntPoint (* (Variables|VoxelizationSetup|GetScene%20Capture%20Resolution) 2)) "RTF_RGBA16f")
                    (else
                      (Variables|InternalRenderTargets|SetVoxelCaptureAtlasRT ""))))
                (:then_1
                  (Utilities|AllocateRTIfNullOrChanged (Variables|InternalRenderTargets|GetVoxelizationRTA) (Utilities|Vector2DtoIntPoint (Variables|Debug|GetRT%20Size)) "RTF_RGBA16f"
                    (:Out
                      (Utilities|AllocateRTIfNullOrChanged (Variables|InternalRenderTargets|GetVoxelizationRTB) (Utilities|Vector2DtoIntPoint (Variables|Debug|GetRT%20Size)) "RTF_RGBA16f"
                        (:Out
                          (Utilities|Array|SetArrayElem _voxelization_rts 0 (Variables|InternalRenderTargets|GetVoxelizationRTA) true)
                          (Utilities|Array|SetArrayElem _voxelization_rts 1 (Variables|InternalRenderTargets|GetVoxelizationRTB) true))))))))))))))

(event Custom|Voxelize%20Blutility
  (bind _actors_to_voxelize (Variables|VoxelizationSetup|GetActors%20to%20Voxelize))
  (bind _output (Utilities|Array|Get(acopy) _actors_to_voxelize))
  (bind (_origin _boxextent) (Collision|GetActorBounds _output false false))
  (bind _box (Variables|Default|GetBox))
  (CallFunction|CaptureMotionVectors)
  (CallFunction|VoxelizeMesh)
  (Transformation|SetWorldLocation _box _origin false false)
  (Components|Box|SetBoxExtent _box _boxextent true)
  (if (Variables|JumpFlooding|GetGenerate%20Distance%20FIeld)
    (Class|JumpFloodComponent3D|JumpFlood (Variables|Default|GetJumpFloodComponent_3D) "NewEnumerator4" (Variables|InternalRenderTargets|GetVoxelizationRTA) (Variables|VoxelizationSetup|GetVolume%20Resolution))
    (bind _result_rt (Class|JumpFloodComponent3D|ComputeGradient (Variables|Default|GetJumpFloodComponent_3D)))
    (Rendering|Material|SetTextureParameterValue (Variables|Debug|GetVoxel%20Raymarch%20Preview%20MID) "RT" _result_rt)))

(event EventTick
  (params (DeltaSeconds :type Float%20%28single-precision%29))
  (bind _actors_to_voxelize (Variables|VoxelizationSetup|GetActors%20to%20Voxelize))
  (bind _output (Utilities|Array|Get(acopy) _actors_to_voxelize))
  (bind (_origin _boxextent) (Collision|GetActorBounds _output false false))
  (bind _box (Variables|Default|GetBox))
  (if (Variables|VoxelizationSetup|GetTick)
    (if (Variables|VoxelizationSetup|GetRender%20Motion%20Vectors)
      (CallFunction|CaptureMotionVectors))
    (CallFunction|VoxelizeMesh)
    (Transformation|SetWorldLocation _box _origin false false)
    (Components|Box|SetBoxExtent _box _boxextent true)
    (if (Variables|JumpFlooding|GetGenerate%20Distance%20FIeld)
      (Class|JumpFloodComponent3D|JumpFlood (Variables|Default|GetJumpFloodComponent_3D) "NewEnumerator4" (Variables|InternalRenderTargets|GetVoxelizationRTA) (Variables|VoxelizationSetup|GetVolume%20Resolution))
      (bind _result_rt (Class|JumpFloodComponent3D|ComputeGradient (Variables|Default|GetJumpFloodComponent_3D)))
      (Rendering|Material|SetTextureParameterValue (Variables|Debug|GetVoxel%20Raymarch%20Preview%20MID) "RT" _result_rt))))

(event EventBeginPlay
  (if (Variables|VoxelizationSetup|GetTick)
    (CallFunction|ManualContructionScript)))

(event Custom|Copy%20Voxel%20Texture
  (bind _rt_to_copy_to (Variables|Default|GetRT%20to%20Copy%20to))
  (bind _returnvalue (Rendering|Material|CreateDynamicMaterialInstance "/Volumetrics/Tools/VoxelizeMeshes/Materials/M_Copy_RT_Simple.M_Copy_RT_Simple" "None" "None" :!class /Script/Engine.KismetMaterialLibrary))
  (Rendering|Material|SetTextureParameterValue _returnvalue "RT" (Variables|InternalRenderTargets|GetVoxelizationRTA))
  (Rendering|ClearRenderTarget2D _rt_to_copy_to)
  (Rendering|DrawMaterialtoRenderTarget _rt_to_copy_to _returnvalue))

(fn CaptureMotionVectors
  (bind _unwraprtb (Variables|InternalRenderTargets|GetUnwrapRTB))
  (bind _unwraprtb_2 (Variables|InternalRenderTargets|GetUnwrapRTB))
  (if (Variables|VoxelizationSetup|GetSwap%20Materials%20for%20Voxelization)
    (CallFunction|SetUnwrapMaterials)
    (Rendering|ClearRenderTarget2D _unwraprtb)
    (Rendering|DrawMaterialtoRenderTarget _unwraprtb (Variables|Debug|GetCopy%20RT%20MID))
    (Rendering|SceneCapture|CaptureScene (Variables|Default|GetMotionVector%20SceneCapture) :!class /Script/Engine.SceneCaptureComponent2D)
    (Math|Integer|IncrementInt (Variables|Default|GetCapture%20Index))
    (else
      (for _array_element (Variables|Debug|GetMaterial%20Reference%20Array)
        (bind _asmaterial_instance_dynamic (Utilities|Casting|CastToMaterialInstanceDynamic _array_element))
        ;[1 1 1] Enable Unwrap Shader Params
        (Rendering|Material|SetScalarParameterValue _asmaterial_instance_dynamic "Unwrap" 1 :!class /Script/Engine.MaterialInstanceDynamic)
        (Rendering|Material|SetScalarParameterValue _asmaterial_instance_dynamic "MotionVectors" 0.0 :!class /Script/Engine.MaterialInstanceDynamic)
        (Rendering|Material|SetVectorParameterValue _asmaterial_instance_dynamic "CapturePos" (Math|Conversions|ToLinearColor(Vector) (Transformation|GetActorLocation)) :!class /Script/Engine.MaterialInstanceDynamic) ; Adjust in case Sim moves
      )
      (Rendering|ClearRenderTarget2D _unwraprtb_2)
      (Rendering|DrawMaterialtoRenderTarget _unwraprtb_2 (Variables|Debug|GetCopy%20RT%20MID))
      (Rendering|SceneCapture|CaptureScene (Variables|Default|GetMotionVector%20SceneCapture) :!class /Script/Engine.SceneCaptureComponent2D)
      (Math|Integer|IncrementInt (Variables|Default|GetCapture%20Index)
        (:%20%20
          (for _array_element_2 (Variables|Debug|GetMaterial%20Reference%20Array)
            (bind _asmaterial_instance_dynamic_2 (Utilities|Casting|CastToMaterialInstanceDynamic _array_element_2))
            ;[1 1 1] Disable Unwrap
            (Rendering|Material|SetScalarParameterValue _asmaterial_instance_dynamic_2 "Unwrap" 0.0 :!class /Script/Engine.MaterialInstanceDynamic)
            (Rendering|Material|SetScalarParameterValue _asmaterial_instance_dynamic_2 "MotionVectors" (Math|Conversions|ToFloat(Boolean) (Variables|VoxelizationSetup|GetShow%20Motion%20On%20Mesh)) :!class /Script/Engine.MaterialInstanceDynamic)))))))

(fn ClearRTs
  (Variables|InternalRenderTargets|SetVoxelizationRTA "")
  (Variables|InternalRenderTargets|SetVoxelizationRTB "")
  (Variables|InternalRenderTargets|SetUnwrapRTA "")
  (Variables|InternalRenderTargets|SetUnwrapRTB "")
  (Variables|InternalRenderTargets|SetVoxelCaptureFrameRT "")
  (Variables|InternalRenderTargets|SetVoxelCaptureAtlasRT ""))

(fn ConfigureSkelmeshMIDs
  (bind _curmat (LocalVariables|Getcurmat))
  (Utilities|Array|Clear (Variables|Debug|GetDisplay%20MotionVector%20MIDs))
  (Utilities|Array|Clear (Variables|Debug|GetUnwrap%20MIDs))
  (for (_array_element _array_index) (Variables|VoxelizationSetup|GetActors%20to%20Voxelize)
    (Utilities|IsValid (Actor|GetComponentbyClass _array_element "/Script/Engine.MeshComponent")
      (:Is%20Valid
        (LocalVariables|Setidx _array_index)
        (for _index (range (Rendering|Material|GetNumMaterials (Actor|GetComponentbyClass _array_element "/Script/Engine.MeshComponent")))
          (if (Variables|VoxelizationSetup|GetSwap%20Materials%20for%20Voxelization)
            (Utilities|Array|Add (LocalVariables|GetMats) (Rendering|Material|GetMaterial (Actor|GetComponentbyClass _array_element "/Script/Engine.MeshComponent") _index))
            (Utilities|FlowControl|Sequence
              (:then_0
                ;[1 1 1] Display Motion Vector MIDs
                (bind _returnvalue (Rendering|Material|CreateDynamicMaterialInstance "/Volumetrics/Tools/VoxelizeMeshes/Materials/M_Display_MotionVectors.M_Display_MotionVectors" "None" "None" :!class /Script/Engine.KismetMaterialLibrary))
                (bind _output_get (LocalVariables|Setcurmat _returnvalue))
                (Utilities|Array|Add (Variables|Debug|GetDisplay%20MotionVector%20MIDs) _output_get))
              (:then_1
                ;[1 1 1] Unwrap MIDs
                (bind _returnvalue_2 (Rendering|Material|CreateDynamicMaterialInstance "/Volumetrics/Tools/VoxelizeMeshes/Materials/M_Unwrap.M_Unwrap" "None" "None" :!class /Script/Engine.KismetMaterialLibrary))
                (bind _output_get_2 (LocalVariables|Setcurmat _returnvalue_2))
                (Utilities|Array|Add (Variables|Debug|GetUnwrap%20MIDs) _output_get_2)))
            (else
              (bind _asmaterial_instance_dynamic (Utilities|Casting|CastToMaterialInstanceDynamic (Rendering|Material|GetMaterial (Actor|GetComponentbyClass _array_element "/Script/Engine.MeshComponent") _index))
                (:then
                  (LocalVariables|Setcurmat _asmaterial_instance_dynamic))
                (:CastFailed
                  (bind _returnvalue_3 (Rendering|Material|CreateDynamicMaterialInstance (Rendering|Material|GetMaterial (Actor|GetComponentbyClass _array_element "/Script/Engine.MeshComponent") _index) "None" "None" :!class /Script/Engine.KismetMaterialLibrary))
                  (bind _output_get_3 (LocalVariables|Setcurmat _returnvalue_3))
                  (Rendering|Material|SetMaterial (Actor|GetComponentbyClass _array_element "/Script/Engine.MeshComponent") _index _output_get_3)))
              (Utilities|Array|Add (LocalVariables|GetMats) (LocalVariables|Getcurmat))))
          ;[1 1 1] Setup MID Params
          (Rendering|Material|SetScalarParameterValue _curmat "idx" (Math|Conversions|ToFloat(Integer) (LocalVariables|Getidx)) :!class /Script/Engine.MaterialInstanceDynamic)
          (Rendering|Material|SetScalarParameterValue _curmat "XYFrames" (Math|Conversions|ToFloat(Integer) (Math|Integer|Max(Integer) (Utilities|Array|Length (Variables|VoxelizationSetup|GetActors%20to%20Voxelize)) 1)) :!class /Script/Engine.MaterialInstanceDynamic)
          (Rendering|Material|SetScalarParameterValue _curmat "VolumeWorldSize" (Variables|VoxelizationSetup|GetVoxelization%20Capture%20Width) :!class /Script/Engine.MaterialInstanceDynamic)
          (Rendering|Material|SetTextureParameterValue _curmat "PreviousWorldPosition" (Variables|InternalRenderTargets|GetUnwrapRTB))
          (Rendering|Material|SetTextureParameterValue _curmat "CurrentWorldPosition" (Variables|InternalRenderTargets|GetUnwrapRTA))
          (Rendering|Material|SetScalarParameterValue _curmat "UVChannel" (Math|Conversions|ToFloat(Integer) (Variables|VoxelizationSetup|GetUV%20Channel%20for%20Motion%20Vectors)) :!class /Script/Engine.MaterialInstanceDynamic)
          (Rendering|Material|SetVectorParameterValue _curmat "CapturePos" (Math|Conversions|ToLinearColor(Vector) (Transformation|GetActorLocation)) :!class /Script/Engine.MaterialInstanceDynamic)))))
  (Variables|Debug|SetMaterial%20Reference%20Array (LocalVariables|GetMats))
  (CallFunction|CreatePoseableMeshes))

(fn ConfiugreVoxelizationMIDs
  (bind (_x _y) (Math|Vector2D|BreakVector2D (Variables|Debug|GetXY%20Frames)))
  (bind _voxel_composite_mid (Variables|Debug|GetVoxel%20Composite%20MID))
  (bind _voxel_frame_copy_mid (Variables|Debug|GetVoxel%20Frame%20Copy%20MID))
  (bind _dilate_uvs_mid (Variables|Debug|GetDilate%20UVs%20MID))
  (Utilities|FlowControl|Sequence
    (:then_0
      (bind _returnvalue (Rendering|Material|CreateDynamicMaterialInstance (Variables|InternalMaterials|GetRaymarch%20Preview%20Material) "None" "None" :!class /Script/Engine.KismetMaterialLibrary))
      (Variables|Debug|SetVoxel%20Raymarch%20Preview%20MID _returnvalue))
    (:then_1
      (bind _returnvalue_2 (Rendering|Material|CreateDynamicMaterialInstance (select (or (Variables|VoxelizationSetup|GetAtlas%20Voxelization) (Variables|VoxelizationSetup|GetUse%20Poseable%20Meshes)) (Variables|InternalMaterials|GetAtlas%20Voxelization%20Material) (Variables|InternalMaterials|GetVoxelization%20Material)) "None" "None" :!class /Script/Engine.KismetMaterialLibrary))
      (Variables|Debug|SetVoxel%20Composite%20MID _returnvalue_2))
    (:then_2
      (bind _returnvalue_3 (Rendering|Material|CreateDynamicMaterialInstance "/Volumetrics/Tools/VoxelizeMeshes/Materials/M_SceneCaptureFrame.M_SceneCaptureFrame" "None" "None" :!class /Script/Engine.KismetMaterialLibrary))
      (Variables|Debug|SetVoxel%20Frame%20Copy%20MID _returnvalue_3))
    (:then_3
      (bind _returnvalue_4 (Rendering|Material|CreateDynamicMaterialInstance "/Volumetrics/Tools/VoxelizeMeshes/Materials/M_DilateUVs.M_DilateUVs" "None" "None" :!class /Script/Engine.KismetMaterialLibrary))
      (Variables|Debug|SetDilate%20UVs%20MID _returnvalue_4))
    (:then_4
      (bind _returnvalue_5 (Rendering|Material|CreateDynamicMaterialInstance "/Volumetrics/Tools/VoxelizeMeshes/Materials/M_Copy_RT_Dilate.M_Copy_RT_Dilate" "None" "None" :!class /Script/Engine.KismetMaterialLibrary))
      (bind _output_get (Variables|Debug|SetCopy%20RT%20MID _returnvalue_5))
      (Rendering|Material|SetTextureParameterValue _output_get "RT" (Variables|InternalRenderTargets|GetUnwrapRTA)))
    (:then_5
      (Rendering|Material|SetTextureParameterValue (Variables|Debug|GetVoxel%20Raymarch%20Preview%20MID) "RT" (Variables|InternalRenderTargets|GetVoxelizationRTA))
      (Rendering|Material|SetVectorParameterValue (Variables|Debug|GetVoxel%20Composite%20MID) "XYFrames" (Math|Color|MakeColor _x _y) :!class /Script/Engine.MaterialInstanceDynamic)
      (Rendering|Material|SetTextureParameterValue (Variables|Debug|GetVoxel%20Composite%20MID) "SceneCapture" (select (Variables|VoxelizationSetup|GetUse%20Poseable%20Meshes) (Variables|InternalRenderTargets|GetVoxelCaptureFrameRT) (select (Variables|VoxelizationSetup|GetAtlas%20Voxelization) (Variables|InternalRenderTargets|GetVoxelCaptureAtlasRT) (Variables|InternalRenderTargets|GetVoxelCaptureFrameRT))))
      (Rendering|Material|SetScalarParameterValue _voxel_composite_mid "VolumeWorldSize" (select (Variables|VoxelizationSetup|GetUse%20Poseable%20Meshes) (Variables|VoxelizationSetup|GetVoxelization%20Capture%20Width) (select (Variables|VoxelizationSetup|GetAtlas%20Voxelization) 1.0 (Variables|VoxelizationSetup|GetVoxelization%20Capture%20Width))) :!class /Script/Engine.MaterialInstanceDynamic)
      (Rendering|Material|SetScalarParameterValue _voxel_composite_mid "SceneCaptureResolution" (Math|Conversions|ToFloat(Integer) (Variables|VoxelizationSetup|GetScene%20Capture%20Resolution)) :!class /Script/Engine.MaterialInstanceDynamic)
      (Rendering|Material|SetVectorParameterValue (Variables|Debug|GetVoxel%20Composite%20MID) "MinRGSizeBA" (Math|Color|MakeColor 0.0 0.0 1) :!class /Script/Engine.MaterialInstanceDynamic)
      (Rendering|Material|SetTextureParameterValue _voxel_frame_copy_mid "RT" (Variables|InternalRenderTargets|GetVoxelCaptureFrameRT))
      (Rendering|Material|SetScalarParameterValue _voxel_frame_copy_mid "VolumeWorldSize" (Variables|VoxelizationSetup|GetVoxelization%20Capture%20Width) :!class /Script/Engine.MaterialInstanceDynamic)
      (Rendering|Material|SetTextureParameterValue _dilate_uvs_mid "RT" (Variables|InternalRenderTargets|GetUnwrapRTA))
      (Rendering|Material|SetScalarParameterValue _dilate_uvs_mid "TextureSize" (Math|Conversions|ToFloat(Integer) (Variables|VoxelizationSetup|GetScene%20Capture%20Resolution)) :!class /Script/Engine.MaterialInstanceDynamic))))

(fn CopySettingsfromOwner
  (Utilities|IsValid (Variables|Default|GetOwning%20FluidSim)
    (:Is%20Valid
      (Variables|VoxelizationSetup|SetVolume%20Resolution (Variables|/Volumetrics/Tools/FluidSimulation/Blueprints/BP_FluidSimulation_01.BP_FluidSimulation_01_C|GetVolume%20Resolution (Variables|Default|GetOwning%20FluidSim))))))

(fn CreatePoseableMeshes
  (if (> (Utilities|Array|Length (Variables|Default|GetPoseable%20Meshes)) 0)
    _
    (elif (Variables|VoxelizationSetup|GetUse%20Poseable%20Meshes)
      (for _array_element (Variables|VoxelizationSetup|GetActors%20to%20Voxelize)
        (Utilities|IsValid (Actor|GetComponentbyClass _array_element "/Script/Engine.SkeletalMeshComponent")
          (:Is%20Valid
            (for _ (range 3)
              (bind _returnvalue (AddComponent|NODE_AddPoseableMeshComponent-1))
              (Components|SkinnedMesh|SetSkinnedAssetandUpdate _returnvalue (Components|SkeletalMesh|GetSkeletalMeshAsset (Actor|GetComponentbyClass _array_element "/Script/Engine.SkeletalMeshComponent")) false)
              (Utilities|Array|Add (Variables|Default|GetPoseable%20Meshes) _returnvalue)
              (Development|SetHiddeninGame _returnvalue (not (Variables|Debug|GetShow%20Poseable%20Meshes)) false)
              (for _index (range (Rendering|Material|GetNumMaterials (Actor|GetComponentbyClass _array_element "/Script/Engine.SkeletalMeshComponent")))
                (Rendering|Material|SetMaterial _returnvalue _index (Rendering|Material|GetMaterial (Actor|GetComponentbyClass _array_element "/Script/Engine.SkeletalMeshComponent") _index))))))))))

(fn GetLocalBounds
  (bind _xy_frames (Variables|Debug|GetXY%20Frames))
  (bind (_x _y) (Math|Vector2D|BreakVector2D _xy_frames))
  (bind _voxelization_capture_width (Variables|VoxelizationSetup|GetVoxelization%20Capture%20Width))
  (bind _returnvalue (/ _voxelization_capture_width _y))
  (bind _returnvalue_2 (Transformation|GetActorLocation))
  (bind _z (.z _returnvalue_2))
  (bind _voxelization_capture_width_2 (Variables|VoxelizationSetup|GetVoxelization%20Capture%20Width))
  (bind _volumesize (Variables|Debug|GetVolumeSize))
  (bind _returnvalue_3 (/ _volumesize _voxelization_capture_width_2))
  (bind _meshboundstovolumeratio (Variables|Debug|GetMeshBoundsToVolumeRatio))
  (bind (_x_2 _y_2) (Math|Vector2D|BreakVector2D _meshboundstovolumeratio))
  (bind _voxelization_capture_width_3 (Variables|VoxelizationSetup|GetVoxelization%20Capture%20Width))
  (bind _returnvalue_4 (* _voxelization_capture_width_3 0.5))
  (bind _box (Variables|Default|GetBox))
  (Components|Box|SetBoxExtent (Variables|Default|GetBox1) (Math|Conversions|ToVector(Float) (* (Variables|VoxelizationSetup|GetVoxelization%20Capture%20Width) 0.5)) false)
  (if (Variables|VoxelizationSetup|GetUse%20Mesh%20Bounds)
    (LocalVariables|SetCurMin "1000000.000000,1000000.000000,1000000.000000")
    (LocalVariables|SetCurMax "-1000000.000000,-1000000.000000,-1000000.000000")
    (for _array_element (Variables|VoxelizationSetup|GetActors%20to%20Voxelize)
      ;[1 1 1] Shared Bounds
      (LocalVariables|SetCurMin (Utilities|VectorMin (LocalVariables|GetCurMin) (- (Collision|GetActorBounds _array_element false false) (.BoxExtent (Collision|GetActorBounds _array_element false false)))))
      (LocalVariables|SetCurMax (Utilities|VectorMax (LocalVariables|GetCurMax) (+ (Collision|GetActorBounds _array_element false false) (.BoxExtent (Collision|GetActorBounds _array_element false false))))))
    (Variables|Debug|Setcurrmin (LocalVariables|GetCurMin))
    (Variables|Debug|Setcurrmax (LocalVariables|GetCurMax))
    (Transformation|SetWorldLocation _box (* (+ (LocalVariables|GetCurMin) (LocalVariables|GetCurMax)) 0.5) false false)
    (Components|Box|SetBoxExtent _box (* (- (LocalVariables|GetCurMax) (LocalVariables|GetCurMin)) 0.5) true)
    (Variables|Debug|SetSpace%20from%20Bottom (Math|Conversions|ToFloat(Integer) (Math|Float|Truncate (/ (- (.z (LocalVariables|GetCurMin)) (- _z _returnvalue_4)) _returnvalue))))
    (Variables|Debug|SetSpace%20from%20Top (Math|Conversions|ToFloat(Integer) (Math|Float|Truncate (/ (- (+ _z _returnvalue_4) (.z (LocalVariables|GetCurMax))) _returnvalue))))
    (Variables|Debug|SetLocal%20Min (Utilities|ConvertPosToLocal0-1 (LocalVariables|GetCurMin)))
    (Variables|Debug|SetLocal%20Size (/ (- (LocalVariables|GetCurMax) (LocalVariables|GetCurMin)) (Variables|VoxelizationSetup|GetVoxelization%20Capture%20Width)))
    (bind _output_get (Variables|BoundsConfig|Setw (- (LocalVariables|GetCurMax) (LocalVariables|GetCurMin))))
    (bind _output_get_2 (Variables|BoundsConfig|Setx (Math|Float|Truncate (.x (* _output_get _returnvalue_3)))))
    (bind _output_get_3 (Variables|BoundsConfig|Sety (Math|Float|Truncate (.y (* _output_get _returnvalue_3)))))
    (Variables|Debug|SetMeshBoundsToVolumeRatio (Math|Vector2D|MakeVector2D (/ (Math|Conversions|ToFloat(Integer) _output_get_2) (.x (Variables|Debug|GetVolumeSize))) (/ (Math|Conversions|ToFloat(Integer) _output_get_3) (.y (Variables|Debug|GetVolumeSize)))))
    (Rendering|Material|SetVectorParameterValue (Variables|Debug|GetVoxel%20Raymarch%20Preview%20MID) "UVScale" (Math|Color|MakeColor _x_2 _y_2 (.x (Variables|Debug|GetLocal%20Min)) (.y (Variables|Debug|GetLocal%20Min))) :!class /Script/Engine.MaterialInstanceDynamic)
    (Rendering|Material|SetVectorParameterValue (Variables|Debug|GetVoxel%20Composite%20MID) "MinRGSizeBA" (Math|Color|MakeColor (.x (Variables|Debug|GetLocal%20Min)) (.y (Variables|Debug|GetLocal%20Min)) _x_2 _y_2) :!class /Script/Engine.MaterialInstanceDynamic)
    (else
      (Variables|Debug|SetSpace%20from%20Bottom 0.0)
      (Variables|Debug|SetSpace%20from%20Top 0.0)
      (Variables|Debug|SetMeshBoundsToVolumeRatio "(X=1.000000,Y=1.000000)")
      (Variables|Debug|SetLocal%20Min "0, 0, 0")
      (Variables|Debug|SetLocal%20Size "1.000000,1.000000,1.000000"))))

(fn ManualContructionScript
  (bind _preview_mesh (Variables|Default|GetPreview%20Mesh))
  (Variables|Default|SetShowOnlyActorList (Variables|VoxelizationSetup|GetActors%20to%20Voxelize))
  (Utilities|Array|Add (Variables|Default|GetShowOnlyActorList) self)
  (CallFunction|CopySettingsFromOwner)
  (CallFunction|ResolutionSetup)
  (CallFunction|CreateRTs)
  (CallFunction|ConfigureSkelmeshMIDs)
  (CallFunction|ConfiugreVoxelizationMIDs)
  (CallFunction|GetLocalBounds)
  (CallFunction|SceneCaptureSetup)
  (CallFunction|CaptureMotionVectors)
  (CallFunction|VoxelizeMesh)
  (if (Variables|Debug|GetShow%20Preview%20Raymarch)
    (Transformation|SetRelativeTransform _preview_mesh (Math|Transform|MakeTransform (Variables|Debug|GetPreview%20Location)) false false)
    (Rendering|Material|SetMaterial _preview_mesh 0 (Variables|Debug|GetVoxel%20Raymarch%20Preview%20MID))
    (Rendering|SetVisibility _preview_mesh true false)))

(fn ResolutionSetup
  (Variables|Debug|SetXY%20Frames (Utilities|Select :Index (Variables|VoxelizationSetup|GetVolume%20Resolution) :NewEnumerator0 "(X=32.000000,Y=1.000000)" :NewEnumerator1 "(X=8.000000,Y=8.000000)" :NewEnumerator2 "(X=10.000000,Y=10.000000)" :NewEnumerator3 "(X=16.000000,Y=8.000000)" :NewEnumerator4 "(X=14.000000,Y=14.000000)" :NewEnumerator5 "(X=16.000000,Y=16.000000)" :NewEnumerator6 "(X=18.000000,Y=18.000000)" :NewEnumerator7 "(X=20.000000,Y=20.000000)"))
  (Variables|Debug|SetRT%20Size (Utilities|Select :Index (Variables|VoxelizationSetup|GetVolume%20Resolution) :NewEnumerator0 "(X=1024.000000,Y=32.000000)" :NewEnumerator1 "(X=512.000000,Y=512.000000)" :NewEnumerator2 "(X=1000.000000,Y=1000.000000)" :NewEnumerator3 "(X=2048.000000,Y=1024.000000)" :NewEnumerator4 "(X=2744.000000,Y=2744.000000)" :NewEnumerator5 "(X=4096.000000,Y=4096.000000)" :NewEnumerator6 "(X=5184.000000,Y=5184.000000)" :NewEnumerator7 "(X=8000.000000,Y=8000.000000)"))
  (Variables|Debug|SetVolumeSize (Math|Conversions|ToVector(Float) (Utilities|Select :Index (Variables|VoxelizationSetup|GetVolume%20Resolution) :NewEnumerator0 32 :NewEnumerator1 64 :NewEnumerator2 100 :NewEnumerator3 128 :NewEnumerator4 196 :NewEnumerator5 256 :NewEnumerator6 324 :NewEnumerator7 400))))

(fn RestoreMaterials
  (for _array_element (Variables|VoxelizationSetup|GetActors%20to%20Voxelize)
    (Utilities|IsValid (Actor|GetComponentbyClass _array_element "/Script/Engine.MeshComponent")
      (:Is%20Valid
        (for _index (range (Rendering|Material|GetNumMaterials (Actor|GetComponentbyClass _array_element "/Script/Engine.MeshComponent")))
          (Rendering|Material|SetMaterial (Actor|GetComponentbyClass _array_element "/Script/Engine.MeshComponent") _index (Utilities|Array|Get(acopy) (Variables|Debug|GetMaterial%20Reference%20Array) (+ _index (LocalVariables|Getcurmatidx))))
          (LocalVariables|Setcurmatidx (+ (- (Rendering|Material|GetNumMaterials (Actor|GetComponentbyClass _array_element "/Script/Engine.MeshComponent")) 1) (LocalVariables|Getcurmatidx))))))))

(fn SceneCaptureSetup
  (Variables|/Script/Engine.SceneCaptureComponent|SetShowOnlyActors (Variables|VoxelizationSetup|GetActors%20to%20Voxelize) (Variables|Default|GetMotionVector%20SceneCapture))
  (Variables|/Script/Engine.SceneCaptureComponent|SetShowOnlyActors (Variables|Default|GetShowOnlyActorList) (Variables|Default|GetVoxelization%20SceneCapture))
  (Variables|/Script/Engine.SceneCaptureComponent2D|SetOrthoWidth (select (Variables|VoxelizationSetup|GetUse%20Poseable%20Meshes) (* (Variables|VoxelizationSetup|GetVoxelization%20Capture%20Width) 2) (Variables|VoxelizationSetup|GetVoxelization%20Capture%20Width)) (Variables|Default|GetVoxelization%20SceneCapture))
  (Variables|/Script/Engine.SceneCaptureComponent2D|SetTextureTarget (Variables|InternalRenderTargets|GetUnwrapRTA) (Variables|Default|GetMotionVector%20SceneCapture))
  (Variables|/Script/Engine.SceneCaptureComponent2D|SetTextureTarget (Variables|InternalRenderTargets|GetVoxelCaptureFrameRT) (Variables|Default|GetVoxelization%20SceneCapture)))

(fn SetMotionVectorMaterials
  (bind _display_motionvector_mids (Variables|Debug|GetDisplay%20MotionVector%20MIDs))
  (bind _currentmatidx (LocalVariables|Getcurrentmatidx))
  (LocalVariables|Setcurrentmatidx 0)
  (for _array_element (Variables|VoxelizationSetup|GetActors%20to%20Voxelize)
    (Utilities|IsValid (Actor|GetComponentbyClass _array_element "/Script/Engine.MeshComponent")
      (:Is%20Valid
        (for _index (range (Rendering|Material|GetNumMaterials (Actor|GetComponentbyClass _array_element "/Script/Engine.MeshComponent")))
          (if (Utilities|IsValid (Utilities|Array|Get(acopy) _display_motionvector_mids (+ _index _currentmatidx)) :!class /Script/Engine.KismetSystemLibrary)
            (Rendering|Material|SetMaterial (Actor|GetComponentbyClass _array_element "/Script/Engine.MeshComponent") _index (Utilities|Array|Get(acopy) _display_motionvector_mids (+ _index _currentmatidx)))
            (Rendering|Material|SetScalarParameterValue (Utilities|Array|Get(acopy) _display_motionvector_mids (+ _index _currentmatidx)) "MVMultiply" (select (> (Variables|Default|GetCapture%20Index) 5) 1.0 0.0) :!class /Script/Engine.MaterialInstanceDynamic)
            (else
              (Development|PrintString (Utilities|String|Append "INVALID MID INDEX: " (Utilities|String|Append (Utilities|String|Append (Utilities|String|ToString(Integer) (+ _index _currentmatidx)) "  Out of MaxIndex: ") (Utilities|String|ToString(Integer) (Utilities|Array|LastIndex (Variables|Debug|GetDisplay%20MotionVector%20MIDs)))))))))
        (LocalVariables|Setcurrentmatidx (+ (Rendering|Material|GetNumMaterials (Actor|GetComponentbyClass _array_element "/Script/Engine.MeshComponent")) (LocalVariables|Getcurrentmatidx)))))))

(fn SetUnwrapMaterials
  (LocalVariables|Setcurrentmatidx 0)
  (for _array_element (Variables|VoxelizationSetup|GetActors%20to%20Voxelize)
    (Utilities|IsValid (Actor|GetComponentbyClass _array_element "/Script/Engine.MeshComponent")
      (:Is%20Valid
        (for _index (range (Rendering|Material|GetNumMaterials (Actor|GetComponentbyClass _array_element "/Script/Engine.MeshComponent")))
          (Rendering|Material|SetMaterial (Actor|GetComponentbyClass _array_element "/Script/Engine.MeshComponent") _index (Utilities|Array|Get(acopy) (Variables|Debug|GetUnwrap%20MIDs) (+ _index (LocalVariables|Getcurrentmatidx)))))
        (LocalVariables|Setcurrentmatidx (+ (Rendering|Material|GetNumMaterials (Actor|GetComponentbyClass _array_element "/Script/Engine.MeshComponent")) (LocalVariables|Getcurrentmatidx)))))))

(fn UpdatePoseableMeshes
  (bind _actors_to_voxelize (Variables|VoxelizationSetup|GetActors%20to%20Voxelize))
  (if (Variables|VoxelizationSetup|GetUse%20Poseable%20Meshes)
    (for (_array_element _array_index) (Variables|Default|GetPoseable%20Meshes)
      (Utilities|IsValid (Utilities|Array|Get(acopy) _actors_to_voxelize (/ _array_index 3))
        (:Is%20Valid
          (Utilities|IsValid (Actor|GetComponentbyClass (Utilities|Array|Get(acopy) _actors_to_voxelize (/ _array_index 3)) "/Script/Engine.SkeletalMeshComponent")
            (:Is%20Valid
              (Components|PoseableMesh|CopyPosefromSkeletalComponent _array_element (Actor|GetComponentbyClass (Utilities|Array|Get(acopy) _actors_to_voxelize (/ _array_index 3)) "/Script/Engine.SkeletalMeshComponent"))
              (Transformation|SetRelativeTransform _array_element (Math|Transform|ComposeTransforms (Math|Transform|MakeRelativeTransform (Transformation|GetWorldTransform (Actor|GetComponentbyClass (Utilities|Array|Get(acopy) _actors_to_voxelize (/ _array_index 3)) "/Script/Engine.SkeletalMeshComponent")) (Transformation|GetActorTransform)) (Math|Transform|MakeTransform (* (.location (Utilities|Array|Get(acopy) (Variables|Default|GetMesh%20Offsets) _array_index)) (Variables|VoxelizationSetup|GetVoxelization%20Capture%20Width)) (.rotation (Utilities|Array|Get(acopy) (Variables|Default|GetMesh%20Offsets) _array_index)) (.scale (Utilities|Array|Get(acopy) (Variables|Default|GetMesh%20Offsets) _array_index)))) false false))))))))

(fn ConstructionScript
  (bind _engineshowflagssetting (Utilities|Struct|MakeEngineShowFlagsSetting "Lighting"))
  (bind _engineshowflagssetting_2 (Utilities|Struct|MakeEngineShowFlagsSetting "Translucency"))
  (bind _engineshowflagssetting_3 (Utilities|Struct|MakeEngineShowFlagsSetting "AmbientOcclusion"))
  (bind _engineshowflagssetting_4 (Utilities|Struct|MakeEngineShowFlagsSetting "AmbientCubemap"))
  (bind _engineshowflagssetting_5 (Utilities|Struct|MakeEngineShowFlagsSetting "ReflectionEnvironment"))
  (bind _engineshowflagssetting_6 (Utilities|Struct|MakeEngineShowFlagsSetting "SkyLighting"))
  (bind _engineshowflagssetting_7 (Utilities|Struct|MakeEngineShowFlagsSetting "ScreenSpaceReflections"))
  (bind _engineshowflagssetting_8 (Utilities|Struct|MakeEngineShowFlagsSetting "DynamicShadows"))
  (bind _engineshowflagssetting_9 (Utilities|Struct|MakeEngineShowFlagsSetting "MotionBlur"))
  (bind _engineshowflagssetting_10 (Utilities|Struct|MakeEngineShowFlagsSetting "Bloom"))
  (bind _engineshowflagssetting_11 (Utilities|Struct|MakeEngineShowFlagsSetting "DeferredLighting"))
  (bind _engineshowflagssetting_12 (Utilities|Struct|MakeEngineShowFlagsSetting "Fog"))
  (bind _array (Utilities|Array|MakeArray _engineshowflagssetting_12 _engineshowflagssetting_11 _engineshowflagssetting_10 _engineshowflagssetting_9 _engineshowflagssetting_8 _engineshowflagssetting_7 _engineshowflagssetting_6 _engineshowflagssetting_5 _engineshowflagssetting_4 _engineshowflagssetting_3 _engineshowflagssetting_2 _engineshowflagssetting))
  ; Configure SceneCapture Showflags
  (Variables|/Script/Engine.SceneCaptureComponent|SetShowFlagSettings _array (Variables|Default|GetVoxelization%20SceneCapture))
  (Variables|/Script/Engine.SceneCaptureComponent|SetShowFlagSettings _array (Variables|Default|GetMotionVector%20SceneCapture))
  (if (Variables|VoxelizationSetup|GetVoxelize%20on%20Construction)
    (CallFunction|ManualContructionScript)
    (if (Variables|JumpFlooding|GetJump%20Flood%20on%20Construction)
      (Class|JumpFloodComponent3D|JumpFlood (Variables|Default|GetJumpFloodComponent_3D) "NewEnumerator4" (Variables|InternalRenderTargets|GetVoxelizationRTA) (Variables|VoxelizationSetup|GetVolume%20Resolution))
      (bind _result_rt (Class|JumpFloodComponent3D|ComputeGradient (Variables|Default|GetJumpFloodComponent_3D)))
      (Rendering|Material|SetTextureParameterValue (Variables|Debug|GetVoxel%20Raymarch%20Preview%20MID) "RT" _result_rt))
    (else
      (CallFunction|ClearRTs))))

(fn VoxelizeMesh
  (bind _current_t (LocalVariables|GetCurrent%20T))
  (bind _rotation (.rotation _current_t))
  (bind (_x _y) (Math|Rotator|BreakRotIntoAxes _rotation))
  (bind _voxel_camera_index (Variables|Default|GetVoxel%20Camera%20Index))
  (bind _returnvalue (+ _voxel_camera_index 1))
  (bind _returnvalue_2 (Math|Integer|%25%28Integer%29 _returnvalue 2))
  (bind _current_t_2 (LocalVariables|GetCurrent%20T))
  (bind _location (.location _current_t_2))
  (bind _rotation_2 (.rotation _current_t_2))
  (bind _returnvalue_3 (Transformation|GetActorLocation))
  (bind _voxelization_capture_width (Variables|VoxelizationSetup|GetVoxelization%20Capture%20Width))
  (bind _returnvalue_4 (* _voxelization_capture_width 0.5))
  (bind _returnvalue_5 (* _location _returnvalue_4))
  (bind _returnvalue_6 (+ _returnvalue_3 _returnvalue_5))
  (bind (_x_2 _y_2) (Math|Vector2D|BreakVector2D (Variables|Debug|GetXY%20Frames)))
  (bind (_x_3 _y_3) (Math|Vector2D|BreakVector2D (Variables|Debug|GetMeshBoundsToVolumeRatio)))
  (bind (_x_4 _y_4) (Math|Vector2D|BreakVector2D (Variables|Debug|GetXY%20Frames)))
  (bind (_x_5 _y_5) (Math|Vector2D|BreakVector2D (Variables|Debug|GetXY%20Frames)))
  (bind (_x_6 _y_6) (Math|Vector2D|BreakVector2D (Variables|Debug|GetXY%20Frames)))
  (bind (_x_7 _y_7) (Math|Vector2D|BreakVector2D (Variables|Debug|GetXY%20Frames)))
  (bind _xy_frames (Variables|Debug|GetXY%20Frames))
  (bind (_x_8 _y_8) (Math|Vector2D|BreakVector2D _xy_frames))
  (bind _returnvalue_7 (/ 1 _x_8))
  (bind _returnvalue_8 (/ 1 _y_8))
  (bind (_x_9 _y_9) (Math|Vector2D|BreakVector2D (Variables|Debug|GetXY%20Frames)))
  (bind _returnvalue_9 (Math|Vector2D|MakeVector2D _returnvalue_7 _returnvalue_8))
  (bind _xy_frames_2 (Variables|Debug|GetXY%20Frames))
  (bind (_x_10 _y_10) (Math|Vector2D|BreakVector2D _xy_frames_2))
  (bind _returnvalue_10 (* _y_10 2))
  (bind _returnvalue_11 (* _x_10 2))
  (bind _voxelization_rts (Variables|InternalRenderTargets|GetVoxelization%20RTs))
  (bind _output (Utilities|Array|Get(acopy) _voxelization_rts _returnvalue_2))
  (bind (_2d_index _vector2d) (Utilities|1DTo2DIndex (Variables|Default|GetVoxel%20Camera%20Index) 2))
  (bind _voxelization_scenecapture (Variables|Default|GetVoxelization%20SceneCapture))
  (bind _voxel_composite_mid (Variables|Debug|GetVoxel%20Composite%20MID))
  (CallFunction|UpdatePoseableMeshes)
  (CallFunction|GetLocalBounds)
  (Rendering|ClearRenderTarget2D (Variables|InternalRenderTargets|GetVoxelCaptureAtlasRT))
  ;[1 1 1] Clear Voxelization RTs
  (for _array_element (Variables|InternalRenderTargets|GetVoxelization%20RTs)
    (Rendering|ClearRenderTarget2D _array_element))
  (if (Variables|VoxelizationSetup|GetRender%20Motion%20Vectors)
    ;[1 1 1] Display Motion Vectors on Meshes
    (if (Variables|VoxelizationSetup|GetSwap%20Materials%20for%20Voxelization)
      (CallFunction|SetMotionVectorMaterials)
      (else
        (for _array_element_2 (Variables|Debug|GetMaterial%20Reference%20Array)
          (bind _asmaterial_instance_dynamic (Utilities|Casting|CastToMaterialInstanceDynamic _array_element_2))
          (Rendering|Material|SetScalarParameterValue _asmaterial_instance_dynamic "MotionVectors" (select (> (Variables|Default|GetCapture%20Index) 5) 1.0 0.0) :!class /Script/Engine.MaterialInstanceDynamic))))
    (else
      _))
  (if (Variables|VoxelizationSetup|GetUse%20Poseable%20Meshes)
    (Utilities|FlowControl|Sequence
      (:then_0
        (LocalVariables|SetCurrent%20T (Utilities|Array|Get(acopy) (Variables|Default|GetCamera%20Transforms)))
        ;[0.3585 0.6536 1] Mesh Capturing
        (Transformation|SetWorldTransform _voxelization_scenecapture (Math|Transform|MakeTransform (+ (* "0.000000,1.000000,-1.000000" _returnvalue_4) _returnvalue_6) _rotation_2) false false)
        (Variables|Default|SetVoxel%20Camera%20Index 0)
        (Rendering|Material|SetTextureParameterValue (Variables|Debug|GetVoxel%20Composite%20MID) "SceneCapture" (Variables|InternalRenderTargets|GetVoxelCaptureFrameRT))
        (Rendering|SceneCapture|CaptureScene (Variables|Default|GetVoxelization%20SceneCapture) :!class /Script/Engine.SceneCaptureComponent2D))
      (:then_1
        _))
    (else
      (for (_array_element_3 _array_index) (Variables|Default|GetCamera%20Transforms)
        (LocalVariables|SetCurrent%20T _array_element_3)
        (Variables|Default|SetVoxel%20Camera%20Index _array_index)
        (Transformation|SetWorldTransform _voxelization_scenecapture (Math|Transform|MakeTransform _returnvalue_6 _rotation_2) false false)
        (Rendering|Material|SetTextureParameterValue (Variables|Debug|GetVoxel%20Composite%20MID) "RT" (Utilities|Array|Get(acopy) (Variables|InternalRenderTargets|GetVoxelization%20RTs) (Math|Integer|%25%28Integer%29 (Variables|Default|GetVoxel%20Camera%20Index) 2)))
        (Rendering|Material|SetVectorParameterValue _voxel_composite_mid "CaptureX" (Math|Conversions|ToLinearColor(Vector) _x) :!class /Script/Engine.MaterialInstanceDynamic)
        (Rendering|Material|SetVectorParameterValue _voxel_composite_mid "CaptureY" (Math|Conversions|ToLinearColor(Vector) _y) :!class /Script/Engine.MaterialInstanceDynamic)
        (Rendering|Material|SetScalarParameterValue (Variables|Debug|GetVoxel%20Composite%20MID) "Invert" (Math|Float|SelectFloat 1 0.0 (> _array_index 1)) :!class /Script/Engine.MaterialInstanceDynamic)
        (Rendering|SceneCapture|CaptureScene (Variables|Default|GetVoxelization%20SceneCapture) :!class /Script/Engine.SceneCaptureComponent2D)
        (if (Variables|VoxelizationSetup|GetAtlas%20Voxelization)
          ;[1 1 1] Copy current Capture to Atlas
          (bind (_canvas _size _context) (Rendering|BeginDrawCanvastoRenderTarget (Variables|InternalRenderTargets|GetVoxelCaptureAtlasRT)))
          (Utilities|IsValid _canvas
            (:Is%20Valid
              (Canvas|DrawMaterial :self _canvas :RenderMaterial (Variables|Debug|GetVoxel%20Frame%20Copy%20MID) :ScreenPosition (* (/ _size 2) _vector2d) :ScreenSize (/ _size 2) :CoordinateSize "(X=1.000000,Y=1.000000)" :Rotation 0.0 :PivotPoint "(X=0.500000,Y=0.500000)")
              (Rendering|EndDrawCanvastoRenderTarget _context)))
          (else
            ;[1 1 1] Render current capture to Volume
            (Rendering|ClearRenderTarget2D _output)
            (Rendering|DrawMaterialtoRenderTarget _output (Variables|Debug|GetVoxel%20Composite%20MID)))))
      (if (Variables|VoxelizationSetup|GetAtlas%20Voxelization)
        (if (Variables|VoxelizationSetup|GetWrite%20Per%20Slice)
          ; Writing Per Slice
          (for _index (range (+ (Math|Float|Truncate (- (/ (* _x_7 _y_7) 4) 1.0)) 1))
            (bind (_canvas_2 _size_2 _context_2) (Rendering|BeginDrawCanvastoRenderTarget (Variables|InternalRenderTargets|GetVoxelizationRTA)))
            (Utilities|IsValid _canvas_2
              (:Is%20Valid
                (Canvas|DrawMaterial _canvas_2 (Variables|Debug|GetVoxel%20Composite%20MID) (* (Math|Vector2D|MakeVector2D (/ (Math|Vector2D|BreakVector2D _size_2) _returnvalue_11) (/ (.Y (Math|Vector2D|BreakVector2D _size_2)) _returnvalue_10)) (Utilities|1dTo2dIndex (* _index 4) (Math|Float|Truncate (Math|Vector2D|BreakVector2D (Variables|Debug|GetXY%20Frames))))) (Math|Vector2D|MakeVector2D (/ (Math|Vector2D|BreakVector2D _size_2) _returnvalue_11) (/ (.Y (Math|Vector2D|BreakVector2D _size_2)) _returnvalue_10)) (Math|Vector2D|MakeVector2D (/ (Math|Vector2D|BreakVector2D (Utilities|1dTo2dIndex (* _index 4) (Math|Float|Truncate (Math|Vector2D|BreakVector2D (Variables|Debug|GetXY%20Frames))))) _x_4) (/ (.Y (Math|Vector2D|BreakVector2D (Utilities|1dTo2dIndex (* _index 4) (Math|Float|Truncate (Math|Vector2D|BreakVector2D (Variables|Debug|GetXY%20Frames)))))) _y_4)) _returnvalue_9 0.0 "(X=0.500000,Y=0.500000)")
                (Canvas|DrawMaterial _canvas_2 (Variables|Debug|GetVoxel%20Composite%20MID) (* (Math|Vector2D|MakeVector2D (/ (Math|Vector2D|BreakVector2D _size_2) _returnvalue_11) (/ (.Y (Math|Vector2D|BreakVector2D _size_2)) _returnvalue_10)) (Utilities|1dTo2dIndex (+ (* _index 4) 1) (Math|Float|Truncate (Math|Vector2D|BreakVector2D (Variables|Debug|GetXY%20Frames))))) (Math|Vector2D|MakeVector2D (/ (Math|Vector2D|BreakVector2D _size_2) _returnvalue_11) (/ (.Y (Math|Vector2D|BreakVector2D _size_2)) _returnvalue_10)) (Math|Vector2D|MakeVector2D (/ (Math|Vector2D|BreakVector2D (Utilities|1dTo2dIndex (+ (* _index 4) 1) (Math|Float|Truncate (Math|Vector2D|BreakVector2D (Variables|Debug|GetXY%20Frames))))) _x_5) (/ (.Y (Math|Vector2D|BreakVector2D (Utilities|1dTo2dIndex (+ (* _index 4) 1) (Math|Float|Truncate (Math|Vector2D|BreakVector2D (Variables|Debug|GetXY%20Frames)))))) _y_5)) _returnvalue_9 0.0 "(X=0.500000,Y=0.500000)")
                (Canvas|DrawMaterial _canvas_2 (Variables|Debug|GetVoxel%20Composite%20MID) (* (Math|Vector2D|MakeVector2D (/ (Math|Vector2D|BreakVector2D _size_2) _returnvalue_11) (/ (.Y (Math|Vector2D|BreakVector2D _size_2)) _returnvalue_10)) (Utilities|1dTo2dIndex (+ (+ (* _index 4) 1) 1) (Math|Float|Truncate (Math|Vector2D|BreakVector2D (Variables|Debug|GetXY%20Frames))))) (Math|Vector2D|MakeVector2D (/ (Math|Vector2D|BreakVector2D _size_2) _returnvalue_11) (/ (.Y (Math|Vector2D|BreakVector2D _size_2)) _returnvalue_10)) (Math|Vector2D|MakeVector2D (/ (Math|Vector2D|BreakVector2D (Utilities|1dTo2dIndex (+ (+ (* _index 4) 1) 1) (Math|Float|Truncate (Math|Vector2D|BreakVector2D (Variables|Debug|GetXY%20Frames))))) _x_6) (/ (.Y (Math|Vector2D|BreakVector2D (Utilities|1dTo2dIndex (+ (+ (* _index 4) 1) 1) (Math|Float|Truncate (Math|Vector2D|BreakVector2D (Variables|Debug|GetXY%20Frames)))))) _y_6)) _returnvalue_9 0.0 "(X=0.500000,Y=0.500000)")
                (Canvas|DrawMaterial _canvas_2 (Variables|Debug|GetVoxel%20Composite%20MID) (* (Math|Vector2D|MakeVector2D (/ (Math|Vector2D|BreakVector2D _size_2) _returnvalue_11) (/ (.Y (Math|Vector2D|BreakVector2D _size_2)) _returnvalue_10)) (Utilities|1dTo2dIndex (+ (+ (+ (* _index 4) 1) 1) 1) (Math|Float|Truncate (Math|Vector2D|BreakVector2D (Variables|Debug|GetXY%20Frames))))) (Math|Vector2D|MakeVector2D (/ (Math|Vector2D|BreakVector2D _size_2) _returnvalue_11) (/ (.Y (Math|Vector2D|BreakVector2D _size_2)) _returnvalue_10)) (Math|Vector2D|MakeVector2D (/ (Math|Vector2D|BreakVector2D (Utilities|1dTo2dIndex (+ (+ (+ (* _index 4) 1) 1) 1) (Math|Float|Truncate (Math|Vector2D|BreakVector2D (Variables|Debug|GetXY%20Frames))))) _x_9) (/ (.Y (Math|Vector2D|BreakVector2D (Utilities|1dTo2dIndex (+ (+ (+ (* _index 4) 1) 1) 1) (Math|Float|Truncate (Math|Vector2D|BreakVector2D (Variables|Debug|GetXY%20Frames)))))) _y_9)) _returnvalue_9 0.0 "(X=0.500000,Y=0.500000)")
                (Rendering|EndDrawCanvastoRenderTarget _context_2))))
          (else
            ;[1 1 1] Atlassed Volume Rendering
            (bind (_canvas_3 _size_3 _context_3) (Rendering|BeginDrawCanvastoRenderTarget (Variables|InternalRenderTargets|GetVoxelizationRTA)))
            (Utilities|IsValid _canvas_3
              (:Is%20Valid
                (Canvas|DrawMaterial _canvas_3 (Variables|Debug|GetVoxel%20Composite%20MID) (Math|Vector2D|MakeVector2D 0.0 (* (* (Variables|Debug|GetSpace%20from%20Bottom) _y_3) (.y (Variables|Debug|GetVolumeSize)))) (* (Math|Vector2D|MakeVector2D (Math|Vector2D|BreakVector2D _size_3) (- (.Y (Math|Vector2D|BreakVector2D _size_3)) (* (.y (Variables|Debug|GetVolumeSize)) (+ (Variables|Debug|GetSpace%20from%20Bottom) (Variables|Debug|GetSpace%20from%20Top))))) (Variables|Debug|GetMeshBoundsToVolumeRatio)) (Math|Vector2D|MakeVector2D 0.0 (/ (Variables|Debug|GetSpace%20from%20Bottom) _y_2)) (Math|Vector2D|MakeVector2D 1 (/ (- (.Y (Math|Vector2D|BreakVector2D _size_3)) (* (.y (Variables|Debug|GetVolumeSize)) (+ (Variables|Debug|GetSpace%20from%20Bottom) (Variables|Debug|GetSpace%20from%20Top)))) (.Y (Math|Vector2D|BreakVector2D _size_3)))) 0.0 "(X=0.500000,Y=0.500000)")
                (Rendering|EndDrawCanvastoRenderTarget _context_3)))))
        ;[1 1 1] Restore Materials
        (if (Variables|VoxelizationSetup|GetShow%20Motion%20On%20Mesh)
          _
          (elif (Variables|VoxelizationSetup|GetSwap%20Materials%20for%20Voxelization)
            (CallFunction|RestoreMaterials)
            (else
              (for _array_element_4 (Variables|Debug|GetMaterial%20Reference%20Array)
                (bind _asmaterial_instance_dynamic_2 (Utilities|Casting|CastToMaterialInstanceDynamic _array_element_4))
                (Rendering|Material|SetScalarParameterValue _asmaterial_instance_dynamic_2 "MotionVectors" 0.0 :!class /Script/Engine.MaterialInstanceDynamic)))))
        (else
          _)))))
