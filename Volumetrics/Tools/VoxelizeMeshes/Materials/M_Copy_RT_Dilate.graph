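# Unlit alpha-composite material for the mesh voxelization tool that, by its name, copies a render target with a one-step dilation.
# It samples the RT texture at the center and four axis neighbors, using BlurSampleOffsets scaled by the inverse of VolumeWorldSize.
# Nested lerps use each sample's RGB as its own alpha, so zero values spread into neighboring pixels.
# It also inverts the center sample's alpha.
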Material /Volumetrics/Tools/VoxelizeMeshes/Materials/M_Copy_RT_Dilate
graph {
 properties {
  bAllowNegativeEmissiveColor=true
  blendMode=BLEND_AlphaComposite
  pixelDepthOffsetMode=PDOM_Legacy
  shadingModel=MSM_Unlit
  usedShadingModels=Unlit
 }
 nodes [
  n1 = TextureSampleParameter2D!(UVs, Level, Apply View MipBias) {
   constMipValue=0
   mipValueMode=TMVM_MipLevel
   parameterName=RT
   samplerType=SAMPLERTYPE_LinearColor
   texture=/Volumetrics/Content/VolumeTextures/Textures/RT_Default
  }
  n2 = TextureCoordinate()
  n3 = MaterialFunctionCall!(OffsetAmount=Divide(B=ScalarParameter() {defaultValue=256, parameterName=VolumeWorldSize})) {materialFunction=/Engine/Functions/Engine_MaterialFunctions01/ImageAdjustment/BlurSampleOffsets}
  n4 = TextureSampleParameter2D(UVs=Add(A=@n2, B=@n3."-1,0")) {
   constMipValue=0
   mipValueMode=TMVM_MipLevel
   parameterName=RT
   samplerType=SAMPLERTYPE_LinearColor
   texture=&RT_Default
  }
  n5 = TextureSampleParameter2D(UVs=Add(A=@n2, B=@n3."0,-1")) {
   constMipValue=0
   mipValueMode=TMVM_MipLevel
   parameterName=RT
   samplerType=SAMPLERTYPE_LinearColor
   texture=&RT_Default
  }
  n6 = TextureSampleParameter2D(UVs=Add(A=@n2, B=@n3."1,0")) {
   constMipValue=0
   mipValueMode=TMVM_MipLevel
   parameterName=RT
   samplerType=SAMPLERTYPE_LinearColor
   texture=&RT_Default
  }
  n7 = TextureSampleParameter2D(UVs=Add(A=@n2, B=@n3."0,1")) {
   constMipValue=0
   mipValueMode=TMVM_MipLevel
   parameterName=RT
   samplerType=SAMPLERTYPE_LinearColor
   texture=&RT_Default
  }
  n8 = LinearInterpolate(A=@n6.RGB, B=@n7.RGB, Alpha=@n6.RGB)
  LinearInterpolate(A=@n1.RGB, B=LinearInterpolate(A=@n4.RGB, B=LinearInterpolate(A=@n5.RGB, B=@n8, Alpha=@n5.RGB), Alpha=@n4.RGB), Alpha=@n1.RGB)
  OneMinus(None=@n1.A)
 ]
}
