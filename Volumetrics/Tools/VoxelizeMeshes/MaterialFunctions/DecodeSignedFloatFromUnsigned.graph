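# Material function in the mesh voxelization tools that decodes a signed float3, named as velocity in its code, from unsigned float storage.
# It splits the input, runs an identical Custom HLSL node on each component and rebuilds the float3.
# The sign rides in the top exponent bit of a 6e5 float: values of 2 or above are scaled by -1/32768 to become negatives, while smaller values pass through.
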MaterialFunction /Volumetrics/Tools/VoxelizeMeshes/MaterialFunctions/DecodeSignedFloatFromUnsigned
graph {nodes [
  MaterialExpressionMaterialFunctionCall_18 = MaterialFunctionCall(Float3=FunctionInput() {bUsePreviewValueAsDefault=true, description=Should be in the -2 to 2 range, inputName=Unsigned Float}) {materialFunction=&BreakOutFloat3Components}
  n1 = Custom!(Velocity=@MaterialExpressionMaterialFunctionCall_18.R) {
   code="// Stored signed value in 6e5 unsigned float\r\n\t// Extract sign from top bit of exponent\r\n\tVelocity *= Velocity >= 2.0 ? (-1.0 / 32768.0) : 1;\r\n\r\nreturn Velocity;"
   description=DecodeNegativeVelocity
   outputType=CMOT_Float1
   inputs [{
     inputName=Velocity
     input {expression=&self:MaterialExpressionMaterialFunctionCall_18}
    }]
  }
  n2 = Custom(Velocity=@MaterialExpressionMaterialFunctionCall_18.G) {
   code="// Stored signed value in 6e5 unsigned float\r\n\t// Extract sign from top bit of exponent\r\n\tVelocity *= Velocity >= 2.0 ? (-1.0 / 32768.0) : 1;\r\n\r\nreturn Velocity;"
   description=DecodeNegativeVelocity
   outputType=CMOT_Float1
   inputs [{
     inputName=Velocity
     input {
      expression=&self:MaterialExpressionMaterialFunctionCall_18
      outputIndex=1
     }
    }]
  }
  n3 = Custom(Velocity=@MaterialExpressionMaterialFunctionCall_18.B) {
   code="// Stored signed value in 6e5 unsigned float\r\n\t// Extract sign from top bit of exponent\r\n\tVelocity *= Velocity >= 2.0 ? (-1.0 / 32768.0) : 1;\r\n\r\nreturn Velocity;"
   description=DecodeNegativeVelocity
   outputType=CMOT_Float1
   inputs [{
     inputName=Velocity
     input {
      expression=&self:MaterialExpressionMaterialFunctionCall_18
      outputIndex=2
     }
    }]
  }
  FunctionOutput(None=MaterialFunctionCall(X=@n1, Y=@n2, Z=@n3) {materialFunction=&MakeFloat3}.Result)
 ]}
