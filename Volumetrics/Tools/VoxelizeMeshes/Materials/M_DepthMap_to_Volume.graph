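# Unlit alpha-composite material for the VoxelizeMeshes tool that projects a scene-capture depth map into a volume laid out as tiled 2D frames.
# EncodeVolumeCoordinates supplies each pixel's volume position, which is projected to pixel-snapped capture UVs and an optionally inverted depth along the capture axis.
# A biased, normalized depth comparison is multiplied with existing render-target alpha to combine occupancy.
# A sphere mask blends capture RGB, apparently velocity, over existing RGB near the captured surface.
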Material /Volumetrics/Tools/VoxelizeMeshes/Materials/M_DepthMap_to_Volume
graph {
 properties {
  bAllowNegativeEmissiveColor=true
  blendMode=BLEND_AlphaComposite
  pixelDepthOffsetMode=PDOM_Legacy
  shadingModel=MSM_Unlit
  usedShadingModels=Unlit
 }
 nodes [
  Constant()
  n1 = TextureSampleParameter2D!(UVs, Level=Constant(), Apply View MipBias) {
   mipValueMode=TMVM_MipLevel
   parameterName=RT
   samplerType=SAMPLERTYPE_LinearColor
   texture=/Volumetrics/Content/VolumeTextures/Textures/RT_Default
  }
  n2 = VectorParameter() {parameterName=CaptureY, defaultValue=<rgba 0 1 0 0>}
  n3 = MaterialFunctionCall!(UVs, XY Frames=ComponentMask(None=VectorParameter() {parameterName=XYFrames, defaultValue=<rgba 16 16 0 0>}.RGB) {g=true, r=true}) {materialFunction=/Volumetrics/Content/VolumeTextures/MaterialFunctions/EncodeVolumeCoordinates}
  n4 = Constant3Vector() {constant=<rgba 0 0 1 0>}
  n5 = MaterialFunctionCall(Float2=AppendVector(A=DotProduct(A=@n2.RGB, B=@n3.Result), B=DotProduct(A=@n4, B=@n3.Result))) {materialFunction=&BreakOutFloat2Components}
  n6 = OneMinus(None=@n5.G)
  n7 = ScalarParameter() {defaultValue=256, parameterName=SceneCaptureResolution}
  n8 = Floor(None=Multiply(A=AppendVector(A=@n5.R, B=@n6), B=@n7) {constB=256})
  n9 = Multiply(A=@n7) {constB=2}
  n10 = TextureSampleParameter2D(UVs=Add(A=Divide(A=@n8) {constB=256}, B=Divide(B=@n9) {constB=512}), Level=Constant()) {
   mipValueMode=TMVM_MipLevel
   parameterName=SceneCapture
   samplerType=SAMPLERTYPE_LinearColor
   texture=&RT_Default
  }
  n11 = ScalarParameter() {parameterName=b2}
  n12 = Add(A=Divide(A=@n10.A, B=ScalarParameter() {defaultValue=200, parameterName=VolumeWorldSize}), B=ScalarParameter() {parameterName=Bias})
  n13 = DotProduct(A=@n3.Result, B=Abs(None=VectorParameter() {parameterName=CaptureX, defaultValue=<rgba 1 0 0 0>}.RGB)) {desc=Fwd}
  n14 = LinearInterpolate(A=@n13, B=OneMinus(None=@n13), Alpha=ScalarParameter() {parameterName=Invert})
  LinearInterpolate(
   A=@n1.RGB
   B=@n10.RGB
   Alpha=SphereMask(A=Add(A=@n11, B=@n12), B=@n14, Radius=ScalarParameter() {defaultValue=0.015, parameterName=Velocity Mask Range})
  )
  Multiply(A=MaterialFunctionCall!(Unsigned Float=@n10.RGB) {materialFunction=/Volumetrics/Tools/VoxelizeMeshes/MaterialFunctions/DecodeSignedFloatFromUnsigned}.Result) {constB=256}
  n15 = Constant()
  n16 = Constant() {r=1}
  OneMinus(None=Multiply(A=@n1.A, B=If(A=@n12, B=@n14, "A > B"=@n15, "A < B"=@n16)))
  ScalarParameter() {parameterName=ss}
 ]
}
