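# Material function for mesh voxelization that samples one cell of a 2x2 SceneCapture atlas for each volume-texture voxel.
# Encoded flipbook volume coordinates, remapped by MinRGSizeBA, are projected along the X/Y axis inputs into texel-centre-snapped UVs.
# Capture alpha, normalized by world size and biased, is compared with optionally inverted voxel depth.
# It outputs captured RGB, a binary depth mask and a thin-shell alpha around the captured surface.
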MaterialFunction /Volumetrics/Tools/VoxelizeMeshes/MaterialFunctions/VoxelizationVolumeSamples
graph {nodes [
  Constant() {r=1}
  n1 = ComponentMask(None=FunctionInput() {bUsePreviewValueAsDefault=true, inputName=Y, sortPriority=1, previewValue=<0 1 0 1>}) {g=true, r=true}
  n2 = VectorParameter() {parameterName=MinRGSizeBA, defaultValue=<rgba 0 0 1 1>}
  n3 = MaterialFunctionCall!(UVs, XY Frames=ComponentMask(None=VectorParameter() {parameterName=XYFrames, defaultValue=<rgba 16 16 0 0>}.RGB) {g=true, r=true}) {materialFunction=/Volumetrics/Content/VolumeTextures/MaterialFunctions/EncodeVolumeCoordinates}
  n4 = Add(
   A=ComponentMask(None=@n2.RGB) {g=true, r=true}
   B=Multiply(A=AppendVector(A=@n2.B, B=@n2.B), B=ComponentMask(None=@n3.Result) {g=true, r=true})
  )
  n5 = MaterialFunctionCall(Float2=AppendVector(A=DotProduct(A=@n1, B=@n4), B=ComponentMask(None=@n3.Result) {b=true})) {materialFunction=&BreakOutFloat2Components}
  n6 = ScalarParameter() {defaultValue=256, parameterName=SceneCaptureResolution}
  n7 = Multiply(A=AppendVector(A=@n5.R, B=OneMinus(None=@n5.G)), B=@n6) {constB=256}
  n8 = Add(A=Divide(A=Floor(None=@n7), B=@n6) {constB=256}, B=Divide(B=Multiply(A=@n6) {constB=2}) {constB=512})
  n9 = Constant() {r=2}
  n10 = MaterialFunctionCall!(1D Index=FunctionInput() {bUsePreviewValueAsDefault=true, inputName=Offset, inputType=FunctionInput_Vector2, sortPriority=2}, Cells XY=@n9) {materialFunction=/Engine/Functions/Engine_MaterialFunctions01/Coordinates/1Dto2DIndex}
  n11 = TextureSampleParameter2D!(UVs=Add(A=Divide(A=@n8), B=Divide(A=@n10.2D Index, B=@n9)), Level=Constant(), Apply View MipBias) {
   mipValueMode=TMVM_MipLevel
   parameterName=SceneCapture
   samplerType=SAMPLERTYPE_LinearColor
   texture=/Volumetrics/Content/VolumeTextures/Textures/RT_Default
  }
  n12 = Divide(A=@n11.A, B=ScalarParameter() {defaultValue=200, parameterName=VolumeWorldSize})
  n13 = Add(A=@n12, B=ScalarParameter() {parameterName=Bias})
  n14 = FunctionInput() {bUsePreviewValueAsDefault=true, inputName=X, previewValue=<1 0 0 1>}
  n15 = DotProduct(A=@n4, B=ComponentMask(None=Abs(None=@n14)) {g=true, r=true}) {desc=Fwd}
  n16 = StaticSwitch(
   True=OneMinus(None=@n15)
   False=@n15
   Value=FunctionInput(Preview=StaticBool()) {bUsePreviewValueAsDefault=true, inputName=Invert Depth, inputType=FunctionInput_StaticBool, sortPriority=3}
  )
  FunctionOutput(None=If(A=@n13, B=@n16, "A > B"=Constant(), "A < B"=Constant() {r=1})) {outputName=Depth Mask, sortPriority=1}
  FunctionOutput(None=@n11.RGB) {outputName=Texture Sample}
  n17 = ScalarParameter() {defaultValue=0.015, parameterName=Velocity Mask Range}
  n18 = Saturate(None=Floor(None=@n12))
  FunctionOutput(None=Multiply(A=SphereMask(A=@n13, B=@n16, Radius=@n17), B=OneMinus(None=@n18))) {outputName=Texture Alpha, sortPriority=1}
 ]}
