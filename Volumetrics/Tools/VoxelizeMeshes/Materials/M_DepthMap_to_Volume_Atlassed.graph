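# Unlit alpha-composite material for the mesh voxelization tool that turns depth map samples into an atlassed volume texture.
# Four VoxelizationVolumeSamples calls use rotated axis pairs, two with an offset and inverted depth, and are layered through chained lerps.
# Use Local Position blends toward volume coordinates encoded from XYFrames, masked by the maximum sample alpha.
# The four depth masks are multiplied and inverted into a combined coverage mask.
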Material /Volumetrics/Tools/VoxelizeMeshes/Materials/M_DepthMap_to_Volume_Atlassed
graph {
 properties {
  bAllowNegativeEmissiveColor=true
  blendMode=BLEND_AlphaComposite
  pixelDepthOffsetMode=PDOM_Legacy
  shadingModel=MSM_Unlit
  usedShadingModels=Unlit
 }
 nodes [
  n1 = MaterialFunctionCall!(X=Constant3Vector() {constant=<rgba 1 0 0 0>}, Y=Constant3Vector() {constant=<rgba 0 1 0 0>}, Offset, Invert Depth) {materialFunction=/Volumetrics/Tools/VoxelizeMeshes/MaterialFunctions/VoxelizationVolumeSamples}
  n2 = MaterialFunctionCall(X=Constant3Vector() {constant=<rgba 0 1 0 0>}, Y=Constant3Vector() {constant=<rgba -1 0 0 0>}) {materialFunction=&VoxelizationVolumeSamples}
  n3 = LinearInterpolate(A=Multiply(A=@n1.Texture Sample, B=@n1.Texture Sample), B=@n2.Texture Sample, Alpha=@n2.Texture Sample)
  n4 = MaterialFunctionCall(
   X=Constant3Vector() {constant=<rgba -1 0 0 0>}
   Y=Constant3Vector() {constant=<rgba 0 -1 0 0>}
   Offset=Constant() {r=3}
   Invert Depth=StaticBool() {value=true}
  ) {materialFunction=&VoxelizationVolumeSamples}
  n5 = MaterialFunctionCall(
   X=Constant3Vector() {constant=<rgba 0 -1 0 0>}
   Y=Constant3Vector() {constant=<rgba 1 0 0 0>}
   Offset=Constant() {r=3}
   Invert Depth=StaticBool() {value=true}
  ) {materialFunction=&VoxelizationVolumeSamples}
  n6 = ComponentMask(None=VectorParameter() {parameterName=XYFrames, defaultValue=<rgba 16 16 0 0>}.RGB) {g=true, r=true}
  n7 = Max(A=@n1.Texture Alpha, B=@n2.Texture Alpha)
  n8 = Max(A=@n4.Texture Alpha, B=@n5.Texture Alpha)
  LinearInterpolate(
   A=LinearInterpolate(A=LinearInterpolate(A=@n3, B=@n4.Texture Sample, Alpha=@n4.Texture Sample), B=@n5.Texture Sample, Alpha=@n5.Texture Sample)
   B=Multiply(
    A=MaterialFunctionCall!(UVs, XY Frames=@n6) {materialFunction=/Volumetrics/Content/VolumeTextures/MaterialFunctions/EncodeVolumeCoordinates}.Result
    B=Max(A=@n7, B=@n8)
   )
   Alpha=ScalarParameter() {parameterName=Use Local Position}
  )
  Multiply(A=MaterialFunctionCall!(Unsigned Float) {materialFunction=/Volumetrics/Tools/VoxelizeMeshes/MaterialFunctions/DecodeSignedFloatFromUnsigned}.Result) {constB=256}
  n9 = Multiply(A=@n1.Depth Mask, B=@n2.Depth Mask)
  OneMinus(None=Multiply(A=Multiply(A=@n9, B=@n4.Depth Mask), B=@n5.Depth Mask))
 ]
}
