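# Unlit alpha-composite material in the voxelize-mesh tools, named for UV dilation, that combines a render target texture with its neighbors.
# BlurSampleOffsets, scaled by the reciprocal of TextureSize, places four one-texel cross samples around the center of the RT texture parameter.
# Nested lerps chain the center and neighbor colors, each using its A-side sample's RGB as alpha.
# The graph shows no connection from the final lerp to a material output.
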Material /Volumetrics/Tools/VoxelizeMeshes/Materials/M_DilateUVs
graph {
 properties {blendMode=BLEND_AlphaComposite, pixelDepthOffsetMode=PDOM_Legacy, shadingModel=MSM_Unlit, usedShadingModels=Unlit}
 nodes [
  n1 = TextureCoordinate()
  n2 = TextureSampleParameter2D(UVs=@n1) {parameterName=RT}
  n3 = MaterialFunctionCall!(OffsetAmount=Divide(B=ScalarParameter() {defaultValue=256, parameterName=TextureSize})) {materialFunction=/Engine/Functions/Engine_MaterialFunctions01/ImageAdjustment/BlurSampleOffsets}
  n4 = TextureSampleParameter2D(UVs=Add(A=@n1, B=@n3."-1,0")) {parameterName=RT}
  n5 = TextureSampleParameter2D(UVs=Add(A=@n1, B=@n3."0,-1")) {parameterName=RT}
  n6 = TextureSampleParameter2D(UVs=Add(A=@n1, B=@n3."1,0")) {parameterName=RT}
  n7 = LinearInterpolate(A=@n6.RGB, B=TextureSampleParameter2D(UVs=Add(A=@n1, B=@n3."0,1")) {parameterName=RT}.RGB, Alpha=@n6.RGB)
  LinearInterpolate(A=@n2.RGB, B=LinearInterpolate(A=@n4.RGB, B=LinearInterpolate(A=@n5.RGB, B=@n7, Alpha=@n5.RGB), Alpha=@n4.RGB), Alpha=@n2.RGB)
 ]
}
