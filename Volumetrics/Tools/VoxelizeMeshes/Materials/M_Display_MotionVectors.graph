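# Unlit display material in the VoxelizeMeshes volumetrics tool that shows motion vectors on meshes, including skeletal and deformer-driven ones.
# The UnPackMotionVectorsWithDilate material function supplies the vectors, which are scaled by the MVMultiply scalar parameter.
# Negative emissive color is allowed, so signed values can pass through unclamped.
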Material /Volumetrics/Tools/VoxelizeMeshes/Materials/M_Display_MotionVectors
graph {
 properties {
  bAllowNegativeEmissiveColor=true
  bUsedWithMeshDeformer=true
  bUsedWithSkeletalMesh=true
  pixelDepthOffsetMode=PDOM_Legacy
  shadingModel=MSM_Unlit
  usedShadingModels=Unlit
 }
 nodes [Multiply(
   A=MaterialFunctionCall() {materialFunction=/Volumetrics/Tools/VoxelizeMeshes/MaterialFunctions/UnPackMotionVectorsWithDilate}.Result
   B=ScalarParameter() {defaultValue=1, parameterName=MVMultiply}
  ) {constA=1}]
}
