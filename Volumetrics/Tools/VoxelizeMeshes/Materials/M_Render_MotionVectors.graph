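# M_Render_MotionVectors is an unlit material in the mesh voxelization tools that, as its name indicates, renders motion vectors from two frames.
# It subtracts the PrevFrame texture's RGB from CurFrame's and scales the difference by a Multiplier parameter.
# Negative emissive color is allowed, so the signed difference can be preserved.
# Both texture parameters default to a tiny render target borrowed from the fluid simulation tools.
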Material /Volumetrics/Tools/VoxelizeMeshes/Materials/M_Render_MotionVectors
graph {
 properties {bAllowNegativeEmissiveColor=true, pixelDepthOffsetMode=PDOM_Legacy, shadingModel=MSM_Unlit, usedShadingModels=Unlit}
 nodes [Multiply(
   A=Subtract(
    A=TextureSampleParameter2D() {parameterName=CurFrame, texture=/Volumetrics/Tools/FluidSimulation/Textures/RT_Tiny}.RGB
    B=TextureSampleParameter2D() {parameterName=PrevFrame, texture=&RT_Tiny}.RGB
   )
   B=ScalarParameter() {defaultValue=1, parameterName=Multiplier}
  )]
}
