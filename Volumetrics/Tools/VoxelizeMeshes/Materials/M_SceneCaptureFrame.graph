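# M_SceneCaptureFrame is an unlit, alpha-composite material in the VoxelizeMeshes volumetrics tool, apparently for processing a scene capture frame.
# It samples the RT render-target parameter and divides its alpha, likely captured depth, by the VolumeWorldSize parameter.
# The result is saturated and inverted, giving a 0–1 value that is highest where the captured alpha is smallest.
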Material /Volumetrics/Tools/VoxelizeMeshes/Materials/M_SceneCaptureFrame
graph {
 properties {
  bAllowNegativeEmissiveColor=true
  blendMode=BLEND_AlphaComposite
  pixelDepthOffsetMode=PDOM_Legacy
  shadingModel=MSM_Unlit
  usedShadingModels=Unlit
 }
 nodes [
  n1 = TextureSampleParameter2D() {parameterName=RT, samplerType=SAMPLERTYPE_LinearColor, texture=/Volumetrics/Content/VolumeTextures/Textures/RT_Default}
  n2 = ScalarParameter() {defaultValue=256, parameterName=VolumeWorldSize}
  OneMinus(None=Saturate(None=Divide(A=@n1.A, B=@n2) {constB=256}))
 ]
}
