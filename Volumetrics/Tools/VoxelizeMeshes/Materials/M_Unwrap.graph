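# Unlit two-sided material for the mesh voxelization tool that unwraps a mesh into UV space for capture, encoding each point's original position.
# An offset flattens vertices to a selectable UV channel (0-2) around CapturePos, scaled by Size and placed in an XYFrames grid cell by idx.
# Pre-offset position is normalized within VolumeWorldSize, quantized to integers and converted per axis by IntToHalfFloat, apparently for precise storage in half-float targets.
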Material /Volumetrics/Tools/VoxelizeMeshes/Materials/M_Unwrap
graph {
 properties {
  bAllowNegativeEmissiveColor=true
  bUsedWithMeshDeformer=true
  bUsedWithSkeletalMesh=true
  pixelDepthOffsetMode=PDOM_Legacy
  shadingModel=MSM_Unlit
  twoSided=true
  usedShadingModels=Unlit
 }
 nodes [
  n1 = VectorParameter() {parameterName=CapturePos}
  n2 = WorldPosition()
  n3 = Ceil(None=ScalarParameter() {parameterName=UVChannel})
  n4 = LinearInterpolate(
   A=LinearInterpolate(A=TextureCoordinate(), B=TextureCoordinate() {coordinateIndex=1}, Alpha=Saturate(None=@n3))
   B=TextureCoordinate() {coordinateIndex=2}
   Alpha=Saturate(None=Subtract(A=@n3))
  )
  n5 = ScalarParameter() {defaultValue=500, parameterName=Size}
  n6 = ScalarParameter() {defaultValue=1, parameterName=XYFrames}
  n7 = MaterialFunctionCall!(1D Index=ScalarParameter() {parameterName=idx}, Cells XY=ScalarParameter() {defaultValue=1, parameterName=XYFrames}) {materialFunction=/Engine/Functions/Engine_MaterialFunctions01/Coordinates/1Dto2DIndex}
  n8 = Add(
   A=Multiply(A=ConstantBiasScale(None=@n4) {bias=-0.5, scale=1}, B=Divide(A=@n5, B=@n6))
   B=Multiply(A=ScalarParameter() {defaultValue=500, parameterName=Size}, B=Subtract(A=@n7.Centered UV Position) {constA=0.5, constB=0.5})
  )
  n9 = Constant()
  Add(A=@n1.RGB, B=Add(A=Multiply(A=@n2.XYZ) {constB=-1}, B=AppendVector(A=@n8, B=@n9)))
  n10 = Subtract(A=WorldPosition() {worldPositionShaderOffset=WPT_ExcludeAllShaderOffsets}.XYZ, B=@n1.RGB)
  n11 = ScalarParameter() {defaultValue=256, parameterName=VolumeWorldSize}
  n12 = Saturate(None=Add(A=Divide(A=@n10, B=@n11) {desc=-0.5 to 0.5}) {constB=0.5, desc=0 to 1})
  n13 = MaterialFunctionCall(Float3=Floor(None=Multiply(A=@n12) {constB=30719})) {materialFunction=&BreakOutFloat3Components}
  MaterialFunctionCall(
   X=MaterialFunctionCall!(Int=@n13.R) {materialFunction=/Volumetrics/Tools/VoxelizeMeshes/MaterialFunctions/IntToHalfFloat}.Result
   Y=MaterialFunctionCall(Int=@n13.G) {materialFunction=&IntToHalfFloat}.Result
   Z=MaterialFunctionCall(Int=@n13.B) {materialFunction=&IntToHalfFloat}.Result
  ) {materialFunction=&MakeFloat3}
 ]
}
