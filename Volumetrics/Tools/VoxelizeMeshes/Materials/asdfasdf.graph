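# An unlit translucent material that appears to be a scratch test of screen-space blur, in the mesh voxelization tools folder.
# The engine SpiralBlur-SceneTexture function samples the scene, with one scalar parameter setting blur distance and another driving both distance and radial step counts.
# Its clamped scene color is offset by a constant one, and that sum is not wired to any material output.
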Material /Volumetrics/Tools/VoxelizeMeshes/Materials/asdfasdf
graph {
 properties {blendMode=BLEND_Translucent, pixelDepthOffsetMode=PDOM_Legacy, shadingModel=MSM_Unlit, usedShadingModels=Unlit}
 nodes [
  n1 = ScalarParameter() {defaultValue=4, parameterName=asdfa}
  Add(
   A=MaterialFunctionCall!(
    Distance=ScalarParameter() {defaultValue=0.1, parameterName=asdf}
    Distance Steps=@n1
    Radial Steps=@n1
    Radial Offset
    TempAA Radial Blur
    TempAA Distance Blur
    Distance Mask
   ) {materialFunction=/Engine/Functions/Engine_MaterialFunctions01/Texturing/SpiralBlur-SceneTexture}.SceneColor clamp to 0
   B=Constant() {r=1}
  )
 ]
}
